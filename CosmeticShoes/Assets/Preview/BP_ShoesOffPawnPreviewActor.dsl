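; Off-pawn preview actor that displays cosmetic shoes in the item preview, with sound, floor and lighting.
; Preview events toggle and re-material the floor plane, with an autotest override.
; On show, a timeline reveals the child prefab's mesh, spawns a Niagara transition and animates a dissolve parameter on it.
; Lighting helpers toggle PC rim and mobile directional lights by platform, performance mode and shadow quality, reacting to settings changes.
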(event EventOnPreviewVisualsBeginLoading
  (CallFunction|HidePreview)
  (Default|SetFloorEnabled (Variables|Default|GetShow%20Floor))
  (Audio|SpawnSoundatLocation "/Game/Sounds/Fort_UI_Sounds/S13_Frontend/BattlePass_ResIn_Cue.BattlePass_ResIn_Cue" (Transformation|GetWorldLocation (Variables|FortItemPreviewOffPawnActor|GetPreviewActorComponent))))

(event EventOnPreviewVisualsSpawned
  (Variables|Default|SetShow%20Floor (Utilities|Struct|BreakCharacterDisplaySettings (Variables|FortItemPreviewActor|GetCurrentDisplayInfo)))
  (Utilities|FlowControl|Sequence
    (:then_0
      (Default|SetFloorEnabled (Variables|Default|GetShow%20Floor)))
    (:then_1
      (Default|SetUPLighting)
      (Default|LightControl true))
    (:then_2
      (bind _returnvalue (Settings|GetGameUserSettings))
      (EventDispatchers|BindEventtoOnGameUserSettingsUINeedsUpdate _returnvalue (delegate Custom|UpdateSettings)))))

(event Custom|UpdateSettings
  (Default|HandleLightingSettingsChanged))

(event EventOnPreviewVisualsShown
  (bind _skeletalmeshtoresin (Variables|Default|GetSkeletalMeshToResIn))
  (Timeline|Timeline PlayFromStart
    (:ShowPawn
      (CallFunction|ShowPreview)
      (bind _asbp_shoes_preview_prefab_asset (Utilities|Casting|CastToBP_ShoesPreviewPrefabAsset (Variables|/Script/Engine.ChildActorComponent|GetChildActor (Variables|FortItemPreviewOffPawnActor|GetPreviewActorComponent))))
      (Class|BPShoesPreviewPrefabAsset|ShowMesh _asbp_shoes_preview_prefab_asset false)
      (Utilities|FlowControl|Delay 0.2)
      (Class|BPShoesPreviewPrefabAsset|ShowMesh _asbp_shoes_preview_prefab_asset true))
    (:PlayResIn
      (Utilities|FlowControl|Delay 0.2)
      (Niagara|SpawnSystematLocation :SystemTemplate "/Game/Effects/Fort_Effects/UI/Transition/P_FrontEndCharacterTransition_Converted.P_FrontEndCharacterTransition_Converted" :Location (+ (Transformation|GetWorldLocation (Variables|FortItemPreviewOffPawnActor|GetPreviewActorComponent)) "-50.000000,-10.000000,0.000000") :Scale "0.500000,0.500000,0.500000" :bAutoDestroy true :bAutoActivate true :PoolingMethod "None" :bPreCullCheck true)
      (Default|GetInstrumentMeshComponents)
      ;[1 1 1] Custom Data needs setting up for this
      (bind (_direction _hiddenamount) (Timeline|Timeline_0 PlayFromStart)
        (:Update
          (for _array_element _skeletalmeshtoresin
            (Rendering|Material|SetScalarParameterValueonMaterials _array_element "NewDissolveGradient" _hiddenamount)))
        (:Finished
          (Utilities|Array|Clear _skeletalmeshtoresin))))) ; Sequence Show / Outro / ResIn
)

(event EventOnSetFloorMaterial
  (params (InMaterialInstance :type Material%20Interface%20Object%20Reference))
  ;[0 0 0] OnSetFloorMaterial
  (bind _output_get (Variables|Default|SetCustomFloorMaterial InMaterialInstance))
  (bind _asmaterial_instance_dynamic (Utilities|Casting|CastToMaterialInstanceDynamic _output_get)
    (:then
      (Rendering|Material|SetScalarParameterValue _asmaterial_instance_dynamic "Floor Metalic" 0.0 :!class /Script/Engine.MaterialInstanceDynamic) ; MF_UI_MPItemShop_Setup
      (Rendering|Material|SetScalarParameterValue _asmaterial_instance_dynamic "Floor Roughness" 0.75 :!class /Script/Engine.MaterialInstanceDynamic) ; MF_UI_MPItemShop_Setup
      (Rendering|Material|SetScalarParameterValue _asmaterial_instance_dynamic "Floor Spec" 0.0 :!class /Script/Engine.MaterialInstanceDynamic)
      (Rendering|Material|SetScalarParameterValue _asmaterial_instance_dynamic "Floor-Metallic" 0.2 :!class /Script/Engine.MaterialInstanceDynamic) ; M_UI_FloorShadow_VaultWorld_Locker
      (Rendering|Material|SetScalarParameterValue _asmaterial_instance_dynamic "Floor-Roughness" 0.75 :!class /Script/Engine.MaterialInstanceDynamic) ; M_UI_FloorShadow_VaultWorld_Locker
    )
    (:CastFailed
      _))
  (Rendering|Material|SetMaterial (Variables|Default|GetPlane) 0 (select (Utilities|IsValid (Variables|Default|GetCustomFloorMaterial) :!class /Script/Engine.KismetSystemLibrary) (Variables|Default|GetCustomFloorMaterial) (Variables|Default|GetDefaultFloorMaterial))))

(event Custom|CustomEvent
  (params (bEnableAutotestBackground :type Boolean))
  (if bEnableAutotestBackground
    (bind _returnvalue (Rendering|Material|CreateDynamicMaterialInstance (Variables|Default|GetAutotestFloorMaterial) "None" "None" :!class /Script/Engine.KismetMaterialLibrary))
    (Rendering|Material|SetMaterial (Variables|Default|GetPlane) 0 _returnvalue)
    (else
      (Rendering|Material|SetMaterial (Variables|Default|GetPlane) 0 (select (Utilities|IsValid (Variables|Default|GetCustomFloorMaterial) :!class /Script/Engine.KismetSystemLibrary) (Variables|Default|GetCustomFloorMaterial) (Variables|Default|GetDefaultFloorMaterial))))))

(fn GetInstrumentMeshComponents
  (Utilities|Array|Clear (Variables|Default|GetSkeletalMeshToResIn))
  (bind _asbp_shoes_preview_prefab_asset (Utilities|Casting|CastToBP_ShoesPreviewPrefabAsset (Variables|/Script/Engine.ChildActorComponent|GetChildActor (Variables|FortItemPreviewOffPawnActor|GetPreviewActorComponent))))
  (Utilities|Array|Add (Variables|Default|GetSkeletalMeshToResIn) (Variables|/Script/CosmeticShoesRuntime.CosmeticShoesPreviewPrefabAsset|GetMesh _asbp_shoes_preview_prefab_asset)))

(fn HandleLightingSettingsChanged
  (bind _returnvalue (TouchPlatform|IsMobileGame))
  (bind _returnvalue_2 (Settings|GetFortGameUserSettings))
  (bind _returnvalue_3 (Class|FortGameUserSettings|IsinPerformanceMode _returnvalue_2))
  (bind _returnvalue_4 (Math|Boolean|ORBoolean _returnvalue _returnvalue_3 false))
  (bind _returnvalue_5 (Settings|GetGameUserSettings))
  (if (Variables|Default|GetIs%20%20active)
    (if (or (== (Settings|GetShadowQuality _returnvalue_5) 0) _returnvalue_4)
      (if _returnvalue_4
        (Default|SwitchPCLighting false)
        (Default|SwitchMobileLighting true)
        (else
          (Default|SwitchPCLighting true)))
      (else
        (Default|SwitchMobileLighting false)
        (Default|SwitchPCLighting true)))))

(fn LightControl
  (params (Active :type Boolean))
  (bind _returnvalue (TouchPlatform|IsMobileGame))
  (bind _returnvalue_2 (TouchPlatform|IsNintendoSwitch))
  (bind _returnvalue_3 (TouchPlatform|IsNintendoSwitch))
  (bind _returnvalue_4 (Settings|GetFortGameUserSettings))
  (bind _returnvalue_5 (Class|FortGameUserSettings|IsinPerformanceMode _returnvalue_4))
  (bind _returnvalue_6 (Math|Boolean|ORBoolean _returnvalue _returnvalue_5 _returnvalue_2))
  (bind _is__active (Variables|Default|GetIs%20%20active))
  (Variables|Default|SetIs%20%20active Active)
  (bind _returnvalue_7 (Settings|GetGameUserSettings))
  (if (Math|Boolean|ORBoolean (== (Settings|GetShadowQuality _returnvalue_7) 0) _returnvalue_6 _returnvalue_3)
    (if _returnvalue_6
      (Default|SwitchMobileLighting _is__active)
      (elif _returnvalue_3
        (Default|SwitchErebusLighting _returnvalue_3))))
  (Default|SwitchPCLighting _is__active))

(fn SetUPLighting
  (Rendering|Components|Light|SetForceCachedShadowsforMovablePrimitives (Variables|Default|GetRimTop_pc) true)
  (Default|SwitchMobileLighting false)
  (Default|SwitchPCLighting false))

(fn SetFloorEnabled
  (params (Show%20Floor :type Boolean))
  (Development|SetHiddeninGame (Variables|Default|GetPlane) (not Show%20Floor) true))

(fn SwitchErebusLighting
  (params (visibility :type Boolean))
  (Rendering|SetVisibility (Variables|Default|GetRimTop_pc) false false)
  (Rendering|SetVisibility (Variables|Default|GetDirectionalLight_mobile) visibility false))

(fn SwitchMobileLighting
  (params (NewParam :type Boolean))
  (Rendering|SetVisibility (Variables|Default|GetRimTop_pc) false false)
  (Rendering|SetVisibility (Variables|Default|GetDirectionalLight_mobile) NewParam false))

(fn SwitchPCLighting
  (params (Visibility :type Boolean))
  (Rendering|SetVisibility (Variables|Default|GetRimTop_pc) Visibility false)
  (Rendering|SetVisibility (Variables|Default|GetDirectionalLight_mobile) false false))

(fn ConstructionScript)
