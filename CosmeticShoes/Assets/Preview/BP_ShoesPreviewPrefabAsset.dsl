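; Preview prefab Blueprint for cosmetic shoes, with a ShowMesh function that shows or hides its mesh by moving it rather than toggling visibility.
; Hiding caches the mesh's current relative location, then moves it to 200 units below its parent's origin.
; Showing restores the cached location, so a repeated hide overwrites that cache with the hidden position.
; BeginPlay, overlap, Tick and the construction script are empty.
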(event EventBeginPlay)

(event Collision|EventActorBeginOverlap
  (params (OtherActor :type Actor%20Object%20Reference)))

(event EventTick
  (params (DeltaSeconds :type Float%20%28single-precision%29)))

(fn ShowMesh
  (params (Show :type Boolean))
  (bind _mesh (Variables|Visual|GetMesh))
  (bind _returnvalue_location (Math|Transform|BreakTransform (Transformation|GetRelativeTransform _mesh)))
  (if Show
    (Transformation|SetRelativeLocation _mesh (Variables|Default|GetCachedMeshLocation) false false)
    (else
      (Variables|Default|SetCachedMeshLocation _returnvalue_location)
      (Transformation|SetRelativeLocation _mesh "0.000000,0.000000,-200.000000" false false))))

(fn ConstructionScript)
